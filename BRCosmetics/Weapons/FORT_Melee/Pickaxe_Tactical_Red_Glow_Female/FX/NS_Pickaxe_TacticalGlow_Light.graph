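# Niagara effect for the Tactical Glow pickaxe that adds a lime-green blade glow and point light, despite the Red folder name.
# One burst-spawn emitter keeps its particle alive, drawing an elongated sprite aligned to local Z plus a light offset along Z.
# An alpha curve indexed by emitter age fades the glow in briefly at start.
# A skeletal mesh user parameter is exposed, but no shown module reads it.
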NiagaraSystem /BRCosmetics/Weapons/FORT_Melee/Pickaxe_Tactical_Red_Glow_Female/FX/NS_Pickaxe_TacticalGlow_Light
dependencies {usedDataInterfaces [&NiagaraDataInterfaceCurve], usedDynamicInputs [], usedModules [&SystemState]}
emitters {Blade_Glow_and_Light {
  emitterUpdateScript [
   &EmitterState(
    Inactive Response=::NewEnumerator1
    Life Cycle Mode=::NewEnumerator1
    Loop Behavior=::NewEnumerator1
    Loop Duration Mode=::NewEnumerator1
    UseLoopDelay=false
   )
   &SpawnBurst_Instantaneous(Use Spawn Probability=false, UseLoopCountLimit=false)
  ]
  particleSpawnScript [
   &InitializeParticle(Color=<rgba 0.4458 1 0 0.1>, Sprite Size=<15 51.5>, Sprite Size Mode=::NewEnumerator4, UsePositionOffset=false, Write Material Random=true)
   &SpriteFacingAndAlignment(Alignment Coordinate Space=::Local, Sprite Alignment=<0 0 1>, WriteAlignment=true, WriteFacing=false)
  ]
  particleUpdateScript [
   &ParticleState(Kill Particles When Lifetime Has Elapsed=false, Loop Particles Lifetime=true)
   &ScaleColor(
    Scale Alpha=&FloatFromCurve(FloatCurve=DataInterface() {propertyValues {Curve {keys [<at 0 0>,<at 0.25 1>]}}}, CurveIndex=$Emitter.Age)
    ScaleA=true
    ScaleRGB=true
   )
   /SetVariables(Particles.LightPosition:NiagaraPosition=<0 0 35>) {isSetParameters=true}
   &FN_Light_Attributes(
    Color=<rgba 0.4458 1 0 1>
    Create Color Attribute=true
    Create Exponent Attribute=true
    Create Position Attribute=true
    Create Radius Attribute=true
    Enable Particle Light Enable=true
    Exponent=8
    Light Enabled=true
    Position:NiagaraPosition=$Particles.LightPosition
    Radius=40
    Scale Color=2
   )
  ]
  rendererClasses [&NiagaraSpriteRendererProperties, &NiagaraLightRendererProperties]
  renderers [
   {rendererClass=&NiagaraSpriteRendererProperties}
   {rendererClass=&NiagaraLightRendererProperties, rendererIndex=1}
  ]
 }}
summary {
 systemName=NS_Pickaxe_TacticalGlow_Light
 emitters [{
   emitterName=Blade_Glow_and_Light
   rendererClasses [&NiagaraSpriteRendererProperties, &NiagaraLightRendererProperties]
  }]
 userVariables [User.SkeletalMEsh {
   type=NiagaraDataInterfaceSkeletalMesh
   defaultValue {
    struct=NiagaraExt_VariableValue_DataInterface
    value {dataInterface=&self:NiagaraDataInterfaceSkeletalMesh_0}
   }
  }]
}
